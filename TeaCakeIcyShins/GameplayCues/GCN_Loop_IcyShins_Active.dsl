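; Looping gameplay cue for the Icy Shins effect, showing shin ice blocks and adding sound and VFX while the target slides.
; On start it flips the right shin block, enables its custom depth, and watches the MovementMode.Sliding tag count through an async task.
; StartSliding spawns looping audio and Niagara at root if absent; StopSliding destroys them.
; Removal spawns expire FX on both calves, stops sliding effects and ends the task.
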(event EventOnLoopingStart(EitherVFX-preferthisversion)
  (params (MyTarget :type Actor%20Object%20Reference) (Parameters :type Gameplay%20Cue%20Parameters%20Structure%20%28by%20ref%29) (ParticleComponents :type Array%20of%20FXSystem%20Component%20Object%20References) (AudioComponents :type Array%20of%20Audio%20Component%20Object%20References))
  (bind _spawnedemitters (Variables|Default|GetSpawnedEmitters))
  (bind _output (Utilities|Array|Get(acopy) _spawnedemitters 1))
  (Utilities|FlowControl|Sequence
    (:then_0
      (Variables|Default|SetTargetActor MyTarget)
      (Variables|Default|SetSpawnedEmitters ParticleComponents)
      ; Flip orientation of the block on the right shin
      (Effects|Components|ParticleSystem|SetVectorParameter _output "BlockOrientation" "0.000000,0.000000,-1.000000")
      (Rendering|SetRenderCustomDepth _output true))
    (:then_1
      ; Start/Stop Sliding, play effects
      (bind (_asynctaskproxy _tagcount) (Ability|Tasks|WaitGameplayTagCountChangedOnActor (Variables|Default|GetTargetActor) "(TagName=\"MovementMode.Sliding\")")
        (:then
          (Variables|Default|SetSlideAbilityTask _asynctaskproxy))
        (:TagCountChanged
          (if (>= _tagcount 1)
            (CallFunction|StartSliding)
            (else
              (CallFunction|StopSliding))))))))

(event EventOnRemoval(EitherVFX-preferthisversion)
  (params (MyTarget :type Actor%20Object%20Reference) (Parameters :type Gameplay%20Cue%20Parameters%20Structure%20%28by%20ref%29) (ParticleComponents :type Array%20of%20FXSystem%20Component%20Object%20References) (AudioComponents :type Array%20of%20Audio%20Component%20Object%20References) (BurstCameraShakeInstance :type Camera%20Shake%20Base%20Object%20Reference) (BurstDecalInstance :type Decal%20Actor%20Object%20Reference))
  (bind _asfort_player_pawn (Utilities|Casting|CastToFortPlayerPawn MyTarget))
  (Effects|SpawnFXSystemAttached :FXSystemAsset (Variables|Default|GetExpireFX) :AttachToComponent (Variables|/Script/Engine.Character|GetMesh _asfort_player_pawn) :AttachPointName "calf_twist_01_l" :Scale "1.000000,1.000000,1.000000" :LocationType "SnapToTarget" :bAutoDestroy true :PoolingMethod "None" :bAutoActivate true)
  (Effects|SpawnFXSystemAttached :FXSystemAsset (Variables|Default|GetExpireFX) :AttachToComponent (Variables|/Script/Engine.Character|GetMesh _asfort_player_pawn) :AttachPointName "calf_twist_01_r" :Scale "1.000000,1.000000,1.000000" :LocationType "SnapToTarget" :bAutoDestroy true :PoolingMethod "None" :bAutoActivate true)
  (CallFunction|StopSliding)
  (Ability|Async|EndAction (Variables|Default|GetSlideAbilityTask)))

(event Custom|StartSliding
  (Utilities|FlowControl|Sequence
    (:then_0
      (Utilities|IsValid (Variables|Default|GetSlideLoopSoundRef)
        (:Is%20Not%20Valid
          (bind _returnvalue (Audio|SpawnSoundAttached :Sound (Variables|Default|GetSlidingLoopingSound) :AttachToComponent (Variables|/Script/Engine.Actor|GetRootComponent (Variables|Default|GetTargetActor)) :AttachPointName "root" :LocationType "SnapToTargetIncludingScale" :bStopWhenAttachedToDestroyed true :VolumeMultiplier 1.0 :PitchMultiplier 1.0 :StartTime 0.0 :bAutoDestroy true))
          (Variables|Default|SetSlideLoopSoundRef _returnvalue))))
    (:then_1
      (Utilities|IsValid (Variables|Default|GetSlideLoopFXRef)
        (:Is%20Not%20Valid
          (bind _returnvalue_2 (Niagara|SpawnSystemAttached (Variables|Default|GetSlideLoopingFX) (Variables|/Script/Engine.Actor|GetRootComponent (Variables|Default|GetTargetActor)) "root" "0, 0, 0" "0, 0, 0" "SnapToTarget" true))
          (Variables|Default|SetSlideLoopFXRef _returnvalue_2))))))

(event Custom|StopSliding
  (bind _slideloopsoundref (Variables|Default|GetSlideLoopSoundRef))
  (Utilities|FlowControl|Sequence
    (:then_0
      (Utilities|IsValid _slideloopsoundref
        (:Is%20Valid
          (Components|DestroyComponent _slideloopsoundref 0))))
    (:then_1
      (Utilities|IsValid (Variables|Default|GetSlideLoopFXRef)
        (:Is%20Valid
          (Components|DestroyComponent (Variables|Default|GetSlideLoopFXRef) 0))))))

(fn ConstructionScript)
